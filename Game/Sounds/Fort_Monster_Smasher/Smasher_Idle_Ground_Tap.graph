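# Sound cue for the Smasher monster's ground tap during idle, classed as creature foley.
# A single hard rock impact wave feeds a Modulator that randomizes pitch and volume on each play, with cue-level multipliers raising both.
# Although a Smasher attenuation asset is referenced, the cue overrides attenuation with its own falloff and overrides concurrency to prevent new instances once the limit is reached.
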SoundCue /Game/Sounds/Fort_Monster_Smasher/Smasher_Idle_Ground_Tap
root=Modulator_0
nodes [
 {id=Modulator_0, children [WavePlayer_0], properties {pitchMax=1.2, pitchMin=0.6, volumeMax=0.5, volumeMin=0.3}}
 {id=WavePlayer_0, properties {soundWaveAssetPtr=/Game/Sounds/Fort_Monster_Smasher/SFX/smasher_rock_hard1}}
]
properties {
 attenuationSettings=/Game/Sounds/Attenuation/Fort_Monster_Smasher_Att
 bOverrideAttenuation=true
 bOverrideConcurrency=true
 maxDistance=3000
 pitchMultiplier=2
 soundClassObject=/Game/Sounds/FortSoundClassesAndModes/Fort_Creatures_Foley
 volumeMultiplier=1.2
 attenuationOverrides {falloffDistance=3000, attenuationShapeExtents {x=0}}
 concurrencyOverrides {resolutionRule=PreventNew}
}
